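# Niagara system for the MP version of the ToonRocks vehicle drift trail cosmetic, trailing fading ribbons and small rock debris from the wheels.
# RibbonsMP spawns quality-scaled skeletal-attached ribbons, while debrisSmall emits colliding flipbook sprites at per-wheel sockets.
# Vehicle speed and frontend preview flags set backward velocity, and both emitters follow source movement with age-based falloff.
# Paint variables drive color, and FN_DisableTireFX masks particles per tire.
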NiagaraSystem /VehicleCosmetics/DriftTrail/DriftTrail_ToonRocks/FX/NS_DriftTrail_ToonRocks_MP
dependencies {
 usedDataInterfaces [
  &NiagaraDataInterfaceCurve
  &NiagaraDataInterfaceSpriteRendererInfo
  &NiagaraDataInterfaceSkeletalMesh
 ]
 usedDynamicInputs []
 usedModules [
  &self:SystemSpawnScript.SystemScriptSource.SystemScriptGraph.NiagaraNodeAssignment_0.SetVariables_#1
  &self:SystemSpawnScript.SystemScriptSource.SystemScriptGraph.NiagaraNodeAssignment_1.SetVariables_#2
  &SystemState
  /VehicleCosmetics/FX/Niagara/Modules/VC_SetVehicleSpeedAttribs
  &FN_UnpackVehicleFlags
  /Game/Effects/Niagara/Modules/Vehicle/FN_UpdateVehicleAttribs
  /Game/Effects/Niagara/Modules/Vehicle/FN_VehicleFlags_Cheatcodes
 ]
}
emitters {
 RibbonsMP {
  emitterUpdateScript [
   &EmitterState(Life Cycle Mode=::NewEnumerator1, Loop Duration=2, UseLoopDelay=false)
   &VC_RateQualityHandler(Epic Spawn Rate=90, High Spawn Rate=60, Low Spawn Rate=30, Medium Spawn Rate=45, Use Custom Rate?=true)
   &VC_SpawnRibbon_v1(Number of Ribbons=2, SpawnRate=$Emitter.Spawn Rate)
  ]
  particleSpawnScript [
   &VC_PaintedVariantsOverride() {enabled=false}
   &VC_InitializeRibbon_v1(
    Color=&MakeCustomLinearColorFromBool(Bool=$User.bIsPainted, True Color=$User.PaintedColor, False Color=<rgba 1 0.2874 0.1022 1>)
    Custom Offset=true
    Enable Skeletal Position=true
    Facing Axis:ENiagaraOrientationAxis=::ZAxis
    Lifetime=1
    Local Vector Offset=&AddVectorToPosition(Position=<0 0 0>, Vector=$User.offset)
    Offset Amount=1
    Ribbon Width=100
   )
   &AddVelocity(Velocity=&VC_Mask_Vectors(
     FrontEnd Value:Vector3f=<-2000 0 0>
     blsFrontEndPreview:NiagaraBool=$User.bIsFrontendPreview
     SpeedMin Value:Vector3f=<-2000 0 0>
     SpeedMax Value:Vector3f=<-1000 0 0>
     Speed Alpha:NiagaraFloat=$System.Speed_Alpha
    )) {module=AddVelocity001}
   /SetVariables(Particles.TwistRibbonLMP=-13, Particles.TwistRibbonRMP=13) {isSetParameters=true}
  ]
  particleUpdateScript [
   &FN_DisableTireFX(Mask Index Attribute:NiagaraInt32=$Particles.SpawnGroup, Mask Index FL=2, Mask Index FR=3, Mask Index RL=0, Mask Index RR=1)
   &ParticleState()
   &ScaleVelocity(
    Coordinate Space:ENiagaraCoordinateSpace=::Simulation
    Velocity Scale=&VectorFromFloat(Value=&FloatFromCurve(FloatCurve=DataInterface() {propertyValues {Curve {keys [<at 0 1>,<at 1 0>]}}}))
   )
   &InheritSourceMovement(
    Applied Position Scale=&VectorFromFloat(Value=&FloatFromCurve(FloatCurve=DataInterface() {propertyValues {Curve {keys [<at 0 1 RCIM_Cubic>, {arriveTangent=-2.0535, interpMode=RCIM_Cubic, leaveTangent=-2.0535, tangentMode=RCTM_User, time=1, value=0}]}}}))
    ApplyForce=false
    ApplyPosition=true
    ApplyPositionFalloff=false
   )
   &SolveForcesAndVelocity(Clamp Velocity=false, Limit Acceleration=false)
   &VC_UpdateRibbon_v1(Enable Skeletal Position=true)
   &ScaleColor(
    Scale Alpha=&Multiply_Float(
     A=&FloatFromCurve(FloatCurve=DataInterface() {propertyValues {Curve {keys [<at 0 1.0113>, {arriveTangent=-2.0226, interpMode=RCIM_Cubic, leaveTangent=-2.0226, time=0.2, value=1}, <at 0.5 0 RCIM_Cubic>]}}})
     B=1
    )
    Scale RGB=&VectorFromFloat(Value=&FloatFromCurve(FloatCurve=DataInterface() {propertyValues {Curve {keys [<at 0 0>,<at 0.04 1.1>]}}}))
    ScaleA=true
    ScaleRGB=true
   )
   &ScaleRibbonWidth(Ribbon Width Scale=&FloatFromCurve(FloatCurve=DataInterface() {propertyValues {Curve {keys [<at 0 1>,<at 0.5 0>]}}}))
   &DynamicMaterialParameters(
    Index 0 Param 1=&MakeCustomFloatFromBool(Bool=$User.blsFrontEndPreview, True Float=10, False Float=2)
    Index 0 Param 2=&RandomRangeFloat(Minimum=0, Maximum=1)
    Index 0 Param 3=1
    Index 0 Param 4=&FloatFromCurve(FloatCurve=DataInterface() {propertyValues {Curve {keys [<at 0 0.5>,<at 3 0.2>]}}})
    Module.Index0Param0Write=true
    Module.Index0Param1Write=true
    Module.Index0Param2Write=true
    Module.Index0Param3Write=true
   )
  ]
  rendererClasses [&NiagaraRibbonRendererProperties]
  renderers [
   {rendererClass=&NiagaraRibbonRendererProperties}
   {rendererClass=&NiagaraRibbonRendererProperties, rendererIndex=1}
  ]
 }
 debrisSmall {
  emitterUpdateScript [
   &EmitterState(Life Cycle Mode=::NewEnumerator1, Loop Duration=2, UseLoopDelay=false)
   &SpawnRate(SpawnRate=32, Use Spawn Probability=false)
  ]
  particleSpawnScript [
   &VC_PaintedVariantsOverride(
    Black=<rgba 0.3741 0.4 0.4 1>
    BurntSienna=<rgba 2.2 0.4609 0.154 3>
    Cobalt=<rgba 0.33 0.5525 3 1>
    Crimson=<rgba 1 0.07 0.132 1>
    DefaultColor=&RandomRangeLinearColor(Minimum=<rgba 1.3 0.3262 0.156 1>, Maximum=<rgba 9 2.258 1.08 1>)
    ForestGreen=<rgba 0.13 1.3 0.13 1>
    Grey=<rgba 0.72 0.7638 0.8 1>
    Lime=<rgba 0.7875 1.5 0.075 1>
    Pink=<rgba 2.5 0.3 1.7441 1>
    Purple=<rgba 0.5779 0.057 0.95 4>
    R1 Black:LinearColor=<rgba 1.64 1.8478 2 3>
    R1 BurntSienna:LinearColor=<rgba 1.8 1.8 1.8 2>
    R1 Cobalt:LinearColor=<rgba 2 2 2 1.5>
    R1 Crimson:LinearColor=<rgba 1.56 1.8139 2 3>
    R1 DefaultColor=<rgba 2 2 2 1.7>
    R1 ForestGreen:LinearColor=<rgba 2.028 2.3581 2.6 2.5>
    R1 Grey:LinearColor=<rgba 2.05 2.3097 2.5 3>
    R1 Lime:LinearColor=<rgba 2.05 2.3097 2.5 1.5>
    R1 Pink:LinearColor=<rgba 1.64 1.8478 2 2>
    R1 Purple:LinearColor=<rgba 1.476 1.663 1.8 2.5>
    R1 Saffron:LinearColor=<rgba 2.05 2.3097 2.5 2>
    R1 SkyBlue:LinearColor=<rgba 1.65 1.9323 2.2 1.8>
    R1 TitaniumWhite:LinearColor=<rgba 2 2.2885 2.5 2>
    Saffron=<rgba 1.6 1.2506 0.112 1>
    SkyBlue=<rgba 0.165 0.7871 1.1 1>
    TitaniumWhite=<rgba 1.1 1.1 1.1 1>
    Use R1 Painted Overrides=true
   )
   &InitializeParticle(
    Color=$Particles.Color
    Lifetime Max=0.8
    Lifetime Min=0.5
    Lifetime Mode=::NewEnumerator1
    Mass=&RandomRangeFloat(Minimum=0, Maximum=1)
    Mass Mode=::NewEnumerator1
    Sprite Rotation Mode=::NewEnumerator2
    Sprite Size Mode=::NewEnumerator1
    Uniform Sprite Size Max=30
    Uniform Sprite Size Min=10
    UsePositionOffset=false
    Write Material Random=true
   )
   &PartitionParticles()
   &SkeletalMeshLocation(
    Bone / Socket Index=&Add_Integer(A=$Particles.PartitionParticles.Partition, B=4)
    Bone / Socket Sampling Mode=::NewEnumerator5
    Interpolate Mesh Sampling=false
    MaskSpawnGroup=false
    Mesh Sampling Type=::NewEnumerator6
    Sampled Position Offset=&AddVectorToPosition(Position=<-50 0 -25>, Vector=Linked() {linkedVariable {name=User.offset001, type=NiagaraPosition}})
    Skeletal Mesh=$System.SkeletalMesh
   )
   &AddVelocity(Velocity=&VC_Mask_Vectors(
     FrontEnd Value:Vector3f=&RandomRangeVector(Minimum=<-600 -50 -100>, Maximum=<-1000 50 -125>)
     blsFrontEndPreview:NiagaraBool=$User.blsFrontEndPreview
     SpeedMin Value:Vector3f=&RandomRangeVector(Minimum=<-600 -50 -100>, Maximum=<-1000 50 -125>)
     SpeedMax Value:Vector3f=&RandomRangeVector(Minimum=<-800 -50 -100>, Maximum=<-800 50 -125>)
     Speed Alpha:NiagaraFloat=$System.Speed_Alpha
    ))
   &ShapeLocation(MaskSpawnGroup=false, Override Seed=false, Sphere Radius=14, UseOwnerScale=true)
  ]
  particleUpdateScript [
   &FN_DisableTireFX(
    Mask Index Attribute:NiagaraInt32=$Particles.PartitionParticles.Partition
    Mask Index FL=2
    Mask Index FR=3
    Mask Index RL=0
    Mask Index RR=1
   )
   &ParticleState()
   &SpriteRotationRate(Rotation Rate=&RandomRangeFloat(Minimum=-1, Maximum=1))
   &ScaleSpriteSize(Uniform Curve Sprite Scale=DataInterface() {propertyValues {Curve {keys [
        {tangentMode=RCTM_User, time=0, value=1}
        {arriveTangent=-0.0405, interpMode=RCIM_Cubic, leaveTangent=-0.0405, tangentMode=RCTM_User, time=0.8, value=1}
        {arriveTangent=-8.0089, interpMode=RCIM_Cubic, leaveTangent=-8.0089, tangentMode=RCTM_User, time=1, value=0}
       ]}}})
   /SetVariables(Particles.StainRockSize:Vector2f=&Multiply_Vector2D_ByFloat(
     Float=&SelectFloatFromArray(
      Array Sampling Mode=::NewEnumerator0
      Direct Array Index=?
      Float Selection Array=DataInterface() {propertyValues {FloatData [0.8,1.1,0.9]}}
      Random Seed=0
     )
     Vector2D=$Particles.SpriteSize
    )) {isSetParameters=true}
   &InheritSourceMovement(
    Applied Position Scale=&Multiply_VectorByFloat(
     Vector=<1 1 1>
     Float=&FloatFromCurve(FloatCurve=DataInterface() {propertyValues {Curve {keys [
          {arriveTangent=-0.3795, interpMode=RCIM_Cubic, leaveTangent=-0.3795, tangentMode=RCTM_User, time=0.3, value=1}
          {arriveTangent=-1.515, tangentMode=RCTM_User, time=1.4, value=0.3}
         ]}}})
    )
    ApplyForce=false
    ApplyPosition=true
    ApplyPositionFalloff=false
   ) {module=InheritSourceMovement001}
   &ScaleVelocity(
    Coordinate Space:ENiagaraCoordinateSpace=::Simulation
    Velocity Scale=&VectorFromFloat(Value=&FloatFromCurve(
      FloatCurve=DataInterface() {propertyValues {Curve {keys [{tangentMode=RCTM_User, time=0, value=0.05}, {tangentMode=RCTM_User, time=0.1, value=0.15}]}}}
      Scale Curve=3
     ))
   )
   &Collision(
    Age Colliding Particles=false
    KillOnCollision=false
    Module.Control Rest State Via Surface Normal=false
    Module.EnableMaxCollisionCount=false
    Module.Randomize Collision Normal Vector=false
   )
   &SolveForcesAndVelocity(Clamp Velocity=false, Limit Acceleration=false)
   &SubUVAnimation(
    Random Start Frame=true
    "Sprite Renderer "=DataInterface() {propertyValues {SpriteRenderer=&self:debrisSmall_1.NiagaraSpriteRendererProperties_1}}
    SubUV Lookup Index Scale:NiagaraFloat=&RandomRangeFloat(Minimum=0.25, Maximum=0.5)
    UseEndFrame=false
    UseStartFrame=false
   ) {module=SubUVAnimation001}
   &SubUVAnimation(
    Random Start Frame=true
    "Sprite Renderer "=DataInterface() {propertyValues {SpriteRenderer=&self:debrisSmall_1.NiagaraSpriteRendererProperties_4}}
    SubUV Lookup Index Scale:NiagaraFloat=&RandomRangeFloat(Minimum=0.25, Maximum=0.5)
    UseEndFrame=false
    UseStartFrame=false
   ) {module=SubUVAnimation002}
  ]
  rendererClasses [&NiagaraSpriteRendererProperties]
  renderers [
   {rendererClass=&NiagaraSpriteRendererProperties}
   {rendererClass=&NiagaraSpriteRendererProperties, rendererIndex=1}
  ]
 }
}
summary {
 systemName=NS_DriftTrail_ToonRocks_MP
 emitters [
  {emitterName=RibbonsMP, rendererClasses [&NiagaraRibbonRendererProperties]}
  {emitterName=debrisSmall, rendererClasses [&NiagaraSpriteRendererProperties]}
 ]
 userVariables [
  User.bIsFrontendPreview {type=NiagaraBool}
  User.bIsPainted {type=NiagaraBool}
  User.blsFrontEndPreview {type=NiagaraBool}
  User.ColorVariantID {type=NiagaraInt32}
  User.DebrisEmissive {type=NiagaraFloat, defaultValue=1, description=""}
  User.GameModeID {type=NiagaraInt32}
  User.IDColor {type=NiagaraInt32}
  User.offset {type=NiagaraPosition, description="", defaultValue {}}
  User.offset {type=Vector3f, description="offset from wheels(set through cosmetic template)", defaultValue {struct=Vector3f, value=<0 0 0>}}
  User.offset001 {type=NiagaraPosition, description="", defaultValue {}}
  User.PaintedColor {type=LinearColor, description="", defaultValue {struct=LinearColor, value=<rgba 1 1 1 1>}}
  User.PaintedScalar {type=NiagaraFloat, defaultValue=1, description=""}
  User.VehicleFlags {type=NiagaraInt32, defaultValue=806289408}
  User.Wheel_Radius_Array {
   type=NiagaraDataInterfaceArrayFloat
   defaultValue {
    struct=NiagaraExt_VariableValue_DataInterface
    value {
     dataInterface=&self:NiagaraDataInterfaceArrayFloat_1
     dataInterfaceClass=&NiagaraDataInterfaceArrayFloat
    }
   }
  }
  User.Wheel_Width_Array {
   type=NiagaraDataInterfaceArrayFloat
   defaultValue {
    struct=NiagaraExt_VariableValue_DataInterface
    value {
     dataInterface=&self:NiagaraDataInterfaceArrayFloat_0
     dataInterfaceClass=&NiagaraDataInterfaceArrayFloat
    }
   }
  }
  User.WheelSize {type=NiagaraFloat, defaultValue=0, description=""}
  User.WheelTouchingArray {
   type=NiagaraDataInterfaceArrayFloat
   defaultValue {
    struct=NiagaraExt_VariableValue_DataInterface
    value {
     dataInterface=&self:NiagaraDataInterfaceArrayFloat_2
     dataInterfaceClass=&NiagaraDataInterfaceArrayFloat
    }
   }
  }
 ]
}
