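# Translucent unlit master material for Niagara ribbon drift trail vehicle effects.
# Optionally flipbooked, panned, distorted base channels are blended, multiplied by a mask and optional panning glow, then edge-faded into emissive and opacity.
# Dynamic parameters drive brightness, channel lerp and distortion; a switch replaces standard opacity with depth fade.
# TrailBend repurposes particle color RGB to drive a sine/cosine ribbon offset, taking tint from CustomColor instead.
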Material /VehicleCosmetics/DriftTrail/DriftTrail_ToonRocks/Materials/M_Trail_TrailsMaster
graph {
 properties {bUsedWithNiagaraRibbons=true, blendMode=BLEND_Translucent, pixelDepthOffsetMode=PDOM_Legacy, shadingModel=MSM_Unlit, usedShadingModels=Unlit}
 nodes [
  n1 = Constant() {r=1}
  MaterialExpressionVertexColor_1 = ParticleColor()
  n2 = TextureCoordinate() {vTiling=0}
  n3 = TextureCoordinate() {uTiling=0}
  n4 = AppendVector(
   A=Multiply(A=ComponentMask(None=@n2) {r=true}, B=ScalarParameter() {defaultValue=1, group=Pan, parameterName=PanGlow_TileU})
   B=Multiply(A=ComponentMask(None=@n3) {g=true}, B=ScalarParameter() {defaultValue=1, group=Pan, parameterName=PanGlow_TileV})
  )
  n5 = Multiply(A=Time(), B=ScalarParameter() {defaultValue=1, group=Pan, parameterName=PanSpeed})
  n6 = StaticSwitchParameter(
   True=Rotator(Coordinate=Panner(Coordinate=@n4, Time=@n5) {speedY=-0.4}, Time=Constant() {r=0.5}) {speed=3.14}
   False=Panner(Coordinate=@n4, Time=@n5) {speedX=-0.4}
  ) {group=Pan, parameterName=RotatePanGlow}
  n7 = Multiply(
   A=ComponentMask(None=TextureCoordinate() {coordinateIndex=1, vTiling=0}) {r=true}
   B=ScalarParameter() {defaultValue=1, group=FlipBook, parameterName=Flipbook_TileU}
  )
  n8 = Time()
  n9 = Constant() {r=1}
  n10 = Panner(
   Coordinate=@n7
   Time=Multiply(
    A=ScalarParameter() {group=FlipBook, parameterName=PanU}
    B=StaticSwitchParameter(True=@n8, False=@n9) {group=FlipBook, parameterName=PanTimeU}
   )
  ) {speedX=0.5}
  n11 = Multiply(
   A=ComponentMask(None=TextureCoordinate() {coordinateIndex=1, uTiling=0}) {g=true}
   B=ScalarParameter() {defaultValue=1, group=FlipBook, parameterName=Flipbook_TileV}
  )
  n12 = Panner(
   Coordinate=@n11
   Time=Multiply(
    A=StaticSwitchParameter(True=@n8, False=@n9) {group=FlipBook, parameterName=PanTimeV}
    B=ScalarParameter() {group=FlipBook, parameterName=PanV}
   )
  ) {speedY=0.5}
  n13 = StaticSwitchParameter(True=AppendVector(A=ComponentMask(None=@n10) {r=true}, B=ComponentMask(None=@n12) {g=true}), False=AppendVector(A=@n7, B=@n11)) {group=FlipBook, parameterName=PanTexture}
  n14 = StaticSwitchParameter(True=Rotator(Coordinate=@n13, Time=Constant() {r=0.5}) {speed=3.14}, False=@n13) {group=FlipBook, parameterName=RotateFlipbook}
  n15 = ScalarParameter() {defaultValue=4, group=FlipBook, parameterName=FB_Columns}
  n16 = ScalarParameter() {defaultValue=4, group=FlipBook, parameterName=FB_Rows}
  n17 = AppendVector(A=@n15, B=@n16)
  n18 = Multiply(A=@n15, B=@n16)
  n19 = Frac(None=Multiply(A=Time(), B=ScalarParameter() {defaultValue=1, group=FlipBook, parameterName=FB_Speed}))
  n20 = Floor(None=Multiply(A=AppendVector(A=@n18, B=@n16), B=AppendVector(A=@n19, B=@n19)))
  n21 = StaticSwitchParameter(True=Add(A=Divide(A=@n14, B=@n17), B=Divide(A=@n20, B=@n17)), False=@n14) {defaultValue=true, group=FlipBook, parameterName=FlipBook}
  n22 = DynamicParameter() {paramNames [Brightness, LerpBlend, UVDistort_Mask, UVDistort]}
  n23 = TextureCoordinate() {uTiling=0.25, vTiling=0.25}
  n24 = ComponentMask(None=TextureSampleParameter2D(UVs=Panner(Coordinate=@n23) {speedX=-0.05, speedY=0.1}) {group=UVDistort, parameterName=UVDistort_Tex, texture=/DealershipSharedLibrary/FX/Textures/Noise/T_VCVFX_Flow03_N}.RGB) {g=true, r=true}
  n25 = Multiply(A=@n22.UVDistort, B=@n24)
  n26 = TextureSampleParameter2D(UVs=StaticSwitchParameter(True=Add(A=@n21, B=@n25), False=@n21) {group=FlipBook, parameterName=UVDistort}) {group=FlipBook, parameterName=BaseTexture, texture=/DealershipSharedLibrary/FX/Textures/TexturePackage/T_VCVFX_ZTITrailFB_RGB}
  n27 = StaticComponentMaskParameter(None=@n26.RGB) {defaultG=true, group=Blend, parameterName=Blend_01}
  n28 = StaticComponentMaskParameter(None=@n26.RGB) {defaultB=true, group=Blend, parameterName=Blend_02}
  n29 = LinearInterpolate(A=@n27, B=@n28, Alpha=@n22.LerpBlend)
  n30 = StaticSwitchParameter(True=Multiply(A=@n27, B=@n28), False=Add(A=@n27, B=@n28)) {group=Blend, parameterName="BlendType(T=M)(F=A)"}
  n31 = Multiply(A=ComponentMask(None=TextureCoordinate() {vTiling=0}) {r=true}, B=ScalarParameter() {defaultValue=1, group=Mask, parameterName=Mask_TileU})
  n32 = ScalarParameter() {group=Mask, parameterName=Mask_PanU}
  n33 = Multiply(A=ComponentMask(None=TextureCoordinate() {uTiling=0}) {g=true}, B=ScalarParameter() {defaultValue=1, group=Mask, parameterName=Mask_TileV})
  n34 = ScalarParameter() {group=Mask, parameterName=Mask_PanV}
  n35 = AppendVector(
   A=ComponentMask(None=Panner(Coordinate=@n31, Time=@n32) {speedX=0.5}) {r=true}
   B=ComponentMask(None=Panner(Coordinate=@n33, Time=@n34) {speedY=0.5}) {g=true}
  )
  n36 = StaticSwitchParameter(True=Rotator(Coordinate=@n35, Time=Constant() {r=0.5}) {speed=3.14}, False=@n35) {group=Mask, parameterName=RotateMask}
  n37 = ScalarParameter() {defaultValue=4, group=Mask, parameterName=M_Columns}
  n38 = ScalarParameter() {defaultValue=4, group=Mask, parameterName=M_Rows}
  n39 = AppendVector(A=@n37, B=@n38)
  n40 = Multiply(A=@n37, B=@n38)
  n41 = Frac(None=Multiply(A=Time(), B=ScalarParameter() {defaultValue=1, group=Mask, parameterName=MaskSpeed}))
  n42 = Floor(None=Multiply(A=AppendVector(A=@n40, B=@n38), B=AppendVector(A=@n41, B=@n41)))
  n43 = StaticSwitchParameter(True=Add(A=Divide(A=@n36, B=@n39), B=Divide(A=@n42, B=@n39)), False=@n36) {defaultValue=true, group=Mask, parameterName=FlipBookMask}
  n44 = Add(A=Multiply(A=@n24, B=@n22.UVDistort_Mask), B=@n43)
  n45 = StaticSwitchParameter(
   True=TextureSampleParameter2D(UVs=StaticSwitchParameter(True=@n44, False=@n43) {group=Mask, parameterName=UVDistort_Mask}) {group=Mask, parameterName=TrailMask, texture=&T_VCVFX_ZTITrailFB_RGB}.RGB
   False=@n26.RGB
  ) {defaultValue=true, group=Mask, parameterName=UseMaskTexture}
  n46 = Multiply(
   A=StaticSwitchParameter(True=StaticSwitchParameter(True=@n29, False=@n30) {group=Blend, parameterName=LerpBlend}, False=Desaturation(None=@n26.RGB)) {group=Blend, parameterName=Blend}
   B=StaticSwitchParameter(True=StaticComponentMaskParameter(None=@n45) {defaultG=true, group=Mask, parameterName=RGB_Mask}, False=@n45) {defaultValue=true, group=Mask, parameterName=UseRGB_Mask?}
  )
  n47 = Multiply(
   A=Multiply(
    A=ScalarParameter() {defaultValue=3, group=Pan, parameterName=PanBrightness}
    B=TextureSampleParameter2D(UVs=@n6) {group=Pan, parameterName=PanGlowTexture, texture=null}.RGB
   )
   B=@n46
  )
  n48 = StaticSwitchParameter(True=StaticSwitchParameter(True=Add(A=@n47, B=@n46), False=@n47) {group=Pan, parameterName=PanAdd}, False=@n46) {group=Pan, parameterName=PanLightning}
  n49 = TextureCoordinate() {vTiling=0}
  n50 = TextureCoordinate() {uTiling=0}
  n51 = AppendVector(
   A=Multiply(A=ComponentMask(None=@n49) {r=true}, B=ScalarParameter() {defaultValue=1, group=GradientFade, parameterName=FadeTileU})
   B=Multiply(A=ComponentMask(None=@n50) {g=true}, B=ScalarParameter() {defaultValue=1, group=GradientFade, parameterName=FadeTileV})
  )
  n52 = StaticSwitchParameter(True=@n51, False=Frac(None=@n51)) {defaultValue=true, group=FlipBook, parameterName=FlipBook}
  n53 = ComponentMask(None=@n52) {r=true}
  n54 = Multiply(
   A=StaticSwitchParameter(True=Multiply(A=@n48, B=@n53), False=@n48) {defaultValue=true, group=GradientFade, parameterName=EdgeFade}
   B=@n22.Brightness
  )
  n55 = ComponentMask(None=@n52) {g=true}
  n56 = OneMinus(None=@n55)
  MaterialExpressionConstantClamp_1 = Saturate(None=Multiply(A=Multiply(A=@n55, B=@n56), B=Constant() {r=8}))
  n57 = StaticSwitchParameter(True=Multiply(A=@n54, B=@MaterialExpressionConstantClamp_1), False=@n54) {defaultValue=true, group=GradientFade, parameterName=EdgeFade}
  MaterialExpressionConstantClamp_0 = Saturate(None=StaticSwitchParameter(
    True=MaterialFunctionCall(Opacity In=@n1) {materialFunction=&MF_VCVFX_DepthFade}.Result
    False=Multiply(A=@MaterialExpressionVertexColor_1.A, B=@n57)
   ) {group=Opacity, parameterName=DBA?})
  n58 = ComponentMask(None=@n52) {r=true}
  n59 = OneMinus(None=@n58)
  n60 = Saturate(None=Multiply(A=@n58, B=Constant() {r=155}))
  n61 = Multiply(A=Multiply(A=@n55, B=Multiply(A=@n59, B=@n60)), B=Constant() {r=8})
  n62 = Multiply(A=ScalarParameter() {defaultValue=8, parameterName=Brightness}, B=@n57)
  n63 = Multiply(
   A=StaticSwitchParameter(True=Multiply(A=@n61, B=@n62), False=@n62) {defaultValue=true, group=GradientFade, parameterName=EdgeFade}
   B=StaticSwitchParameter(
    True=VectorParameter() {parameterName=CustomColor, defaultValue=<rgba 0.9447 0.9447 0.9447 1>}.RGB
    False=@MaterialExpressionVertexColor_1.RGB
   ) {group=TrailBend, parameterName=TrailBend}
  )
  MaterialExpressionConstantClamp_3 = Saturate(None=@n63)
  MaterialFunctionCall(Emissive=StaticSwitchParameter(True=@MaterialExpressionConstantClamp_3, False=@n63) {parameterName=Clamped?}) {materialFunction=&MF_FX_EmissiveAdjustments}
  MaterialExpressionVertexColor_0 = ParticleColor()
  n64 = Multiply(A=@MaterialExpressionVertexColor_0.G, B=OneMinus(None=Time()))
  n65 = ComponentMask(None=Multiply(A=TextureCoordinate() {coordinateIndex=1, uTiling=2}, B=@MaterialExpressionVertexColor_0.R)) {r=true}
  n66 = Add(A=@n64, B=@n65)
  n67 = Add(A=Cosine(None=@n66), B=Cosine(None=@n66) {period=2})
  n68 = Add(A=@n65, B=@n64)
  n69 = AppendVector(A=@n67, B=Add(A=Sine(None=@n68), B=Sine(None=@n68) {period=2}))
  n70 = Multiply(A=Transform(None=AppendVector(A=@n69, B=@n67)), B=@MaterialExpressionVertexColor_0.B)
  n71 = Constant()
  Transform(None=Add(
    A=VectorParameter() {parameterName=RibbonWPO, defaultValue=<rgba 0 0 0 1>}.RGB
    B=StaticSwitchParameter(True=@n70, False=@n71) {group=TrailBend, parameterName=TrailBend}
   )) {transformSourceType=TRANSFORMSOURCE_Local}
 ]
}
